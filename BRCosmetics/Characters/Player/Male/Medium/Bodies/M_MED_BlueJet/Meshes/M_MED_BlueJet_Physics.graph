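# Physics asset for the BlueJet medium male player body cosmetic, covering its dangling buckles and chain.
# Left and right buckles are three-capsule chains with tight angular limits, anchored through small master bodies to spine_05.
# The chain is a single box with narrow swing2 and twist limits, anchored through its master body to the pelvis.
# Capsules cover only pelvis, spine and left thigh, with free angular motion on their constraints.
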PhysicsAsset /BRCosmetics/Characters/Player/Male/Medium/Bodies/M_MED_BlueJet/Meshes/M_MED_BlueJet_Physics
bodies {
 dyn_L_buckle_1_jnt [{shapeType=Capsule, length=2, radius=3.75, shapeName=dyn_L_buckle_1_jnt_capsule, center=<4.7515 -3.4095 0>, rotation=<rot 90 90 90.0001>}]
 dyn_L_buckle_2_jnt [{shapeType=Capsule, length=2, radius=3.75, shapeName=dyn_L_buckle_2_jnt_capsule, center=<4.7515 -2.6868 0>, rotation=<rot 90 179.9999 -180>}]
 dyn_L_buckle_3_jnt [{shapeType=Capsule, length=2, radius=3.75, shapeName=dyn_L_buckle_3_jnt_capsule, center=<4.7515 -2.6868 0>, rotation=<rot 90 0.0001 0.0002>}]
 dyn_R_buckle_1_jnt [{shapeType=Capsule, length=2, radius=3.75, shapeName=dyn_R_buckle_1_jnt_capsule, center=<-5.439 3.4 0>, rotation=<rot 90 160.5288 160.5289>}]
 dyn_R_buckle_2_jnt [{shapeType=Capsule, length=2, radius=3.75, shapeName=dyn_R_buckle_2_jnt_capsule, center=<-5.439 2.6 0>, rotation=<rot 90 179.9209 179.921>}]
 dyn_R_buckle_3_jnt [{shapeType=Capsule, length=2, radius=3.75, shapeName=dyn_R_buckle_3_jnt_capsule, center=<-5.439 2.7 0>, rotation=<rot 90 179.8417 179.8418>}]
 dyn_chain_jnt [{shapeType=Box, extentX=8.5, extentY=4.5, extentZ=13, shapeName=dyn_chain_jnt_box, center=<3.9311 -3.0107 0>, rotation=<rot 0 -10 0>}]
 master_L_buckle_jnt [{shapeType=Capsule, length=1, radius=1, shapeName=master_L_buckle_jnt_capsule}]
 master_R_buckle_jnt [{shapeType=Capsule, length=1, radius=1, shapeName=master_R_buckle_jnt_capsule}]
 master_chain_jnt [{
   shapeType=Capsule
   length=1
   radius=1
   shapeName=master_chain_jnt_capsule
   center=<1.0223 0.0146 0.5775>
   rotation=<rot 0 -119.998 -166.3769>
  }]
 pelvis [{shapeType=Capsule, length=5.75, radius=11.75, shapeName=pelvis_capsule, center=<-10.596 -1.5588 0>}]
 spine_01 [{shapeType=Capsule, length=6.75, radius=11.5, shapeName=spine_01_capsule, center=<-6.2038 0 0>}]
 spine_02 [{shapeType=Capsule, length=7.5, radius=12.75, shapeName=spine_02_capsule, center=<-7.7763 0 0>}]
 spine_03 [{shapeType=Capsule, length=8.75, radius=12.5, shapeName=spine_03_capsule, center=<-6.2038 0 0>}]
 spine_04 [{shapeType=Capsule, length=9, radius=12.5, shapeName=spine_04_capsule, center=<-6.2038 0 0>}]
 spine_05 [{shapeType=Capsule, length=12.75, radius=12.25, shapeName=spine_05_capsule, center=<-6.2038 0 0>}]
 thigh_l [{shapeType=Capsule, length=16.75, radius=6.75, shapeName=thigh_l_capsule, center=<-18.0954 1.6355 -0.6341>, rotation=<rot -80 -0.0002 0>}]
 thigh_twist_01_l [{shapeType=Capsule, length=9, radius=6.75, shapeName=thigh_l_capsule, center=<5.4209 3.0622 -0.6811>, rotation=<rot -90 -19.471 19.4706>}]
}
constraints [
 {bone1Name=dyn_R_buckle_3_jnt, bone2Name=dyn_R_buckle_2_jnt, swing1LimitDegrees=10, swing2LimitDegrees=10, twistLimitDegrees=2}
 {bone1Name=dyn_R_buckle_2_jnt, bone2Name=dyn_R_buckle_1_jnt, swing1LimitDegrees=10, swing2LimitDegrees=10, twistLimitDegrees=2}
 {bone1Name=dyn_L_buckle_3_jnt, bone2Name=dyn_L_buckle_2_jnt, swing1LimitDegrees=10, swing2LimitDegrees=10, twistLimitDegrees=2}
 {bone1Name=dyn_L_buckle_2_jnt, bone2Name=dyn_L_buckle_1_jnt, swing1LimitDegrees=10, swing2LimitDegrees=10, twistLimitDegrees=2}
 {bone1Name=dyn_R_buckle_1_jnt, bone2Name=master_R_buckle_jnt, swing1LimitDegrees=10, swing2LimitDegrees=10, twistLimitDegrees=2}
 {bone1Name=dyn_L_buckle_1_jnt, bone2Name=master_L_buckle_jnt, swing1LimitDegrees=10, swing2LimitDegrees=10, twistLimitDegrees=2}
 {
  bone1Name=master_L_buckle_jnt
  bone2Name=spine_05
  swing1LimitDegrees=45
  swing1Motion=Free
  swing2LimitDegrees=45
  swing2Motion=Free
  twistLimitDegrees=45
  twistMotion=Free
 }
 {
  bone1Name=master_R_buckle_jnt
  bone2Name=spine_05
  swing1LimitDegrees=45
  swing1Motion=Free
  swing2LimitDegrees=45
  swing2Motion=Free
  twistLimitDegrees=45
  twistMotion=Free
 }
 {
  bone1Name=spine_05
  bone2Name=spine_04
  swing1LimitDegrees=45
  swing1Motion=Free
  swing2LimitDegrees=45
  swing2Motion=Free
  twistLimitDegrees=45
  twistMotion=Free
 }
 {
  bone1Name=spine_02
  bone2Name=spine_01
  swing1LimitDegrees=45
  swing1Motion=Free
  swing2LimitDegrees=45
  swing2Motion=Free
  twistLimitDegrees=45
  twistMotion=Free
 }
 {
  bone1Name=spine_03
  bone2Name=spine_02
  swing1LimitDegrees=45
  swing1Motion=Free
  swing2LimitDegrees=45
  swing2Motion=Free
  twistLimitDegrees=45
  twistMotion=Free
 }
 {
  bone1Name=spine_04
  bone2Name=spine_03
  swing1LimitDegrees=45
  swing1Motion=Free
  swing2LimitDegrees=45
  swing2Motion=Free
  twistLimitDegrees=45
  twistMotion=Free
 }
 {
  bone1Name=master_chain_jnt
  bone2Name=pelvis
  swing1LimitDegrees=45
  swing1Motion=Free
  swing2LimitDegrees=45
  swing2Motion=Free
  twistLimitDegrees=45
  twistMotion=Free
 }
 {bone1Name=dyn_chain_jnt, bone2Name=master_chain_jnt, swing1LimitDegrees=45, swing2LimitDegrees=3, twistLimitDegrees=1}
 {
  bone1Name=thigh_twist_01_l
  bone2Name=thigh_l
  swing1LimitDegrees=45
  swing1Motion=Free
  swing2LimitDegrees=45
  swing2Motion=Free
  twistLimitDegrees=45
  twistMotion=Free
 }
]
