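# Secondary-motion AnimBlueprint for the M_MED_BlueJet male body's buckles, D-rings and chain.
# The parent mesh's pose passes through a deform Control Rig, then weighted constraints place the accessory master bones against torso and hip bones.
# RigidBody and AnimDynamics nodes add physics, followed by fixed and state-driven bone fixes.
# Update converts crouch, emote and front-end flags to floats; crouch and emote adjust the chain, while disabled dynamics offsets the buckles.
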AnimBlueprint /BRCosmetics/Characters/Player/Male/Medium/Bodies/M_MED_BlueJet/Meshes/M_MED_BlueJet_AnimBP
target_skeleton=&/M_MED_BlueJet_Skeleton
graphs [
 AnimGraph {nodes [
   n1 = Control Rig(Source=Copy Pose From Mesh() {bUseAttachedParent=true}) {
    alphaBoolBlend {blendOption=Linear}
    controlRigAssetReference {blueprintRigClass=&/M_MED_DeformRig_BlueJet_ControlRig.M_MED_DeformRig_BlueJet_ControlRig_C}
   }
   n2 = "Constraint - master_R_buckle_jnt (PPPP)"!(
    ComponentPose=Local To Component(LocalPose=@n1)
    Alpha=1
    ConstraintWeights_0=0.1033
    ConstraintWeights_1:float=0.2405
    ConstraintWeights_2:float=0.1362
    ConstraintWeights_3:float=0.52
   ) -> (Pose) {
    alphaBoolBlend {blendOption=Linear}
    boneToModify {boneName=master_R_buckle_jnt}
    constraintSetup [{targetBone {boneName=spine_01}}, {targetBone {boneName=spine_02}}, {targetBone {boneName=spine_04}}, {targetBone {boneName=spine_03}}]
    constraintWeights [1,1,1,1]
   }
   n3 = "Constraint - master_L_buckle_jnt (PPPPPP)"!(
    ComponentPose=@n2
    Alpha=1
    ConstraintWeights_0=0.0443
    ConstraintWeights_1:float=0.0949
    ConstraintWeights_2:float=0.0489
    ConstraintWeights_3:float=0.0179
    ConstraintWeights_4:float=0.7765
    ConstraintWeights_5:float=0.0175
   ) -> (Pose) {
    alphaBoolBlend {blendOption=Linear}
    boneToModify {boneName=master_L_buckle_jnt}
    constraintSetup [
     {targetBone {boneName=spine_01}}
     {targetBone {boneName=spine_02}}
     {targetBone {boneName=spine_04}}
     {targetBone {boneName=pelvis}}
     {targetBone {boneName=spine_03}}
     {targetBone {boneName=deform_pec_l}}
    ]
    constraintWeights [1,1,1,1,1,1]
   }
   n4 = "Constraint - master_L_dring_jnt (PPPPPPP)"!(
    ComponentPose=@n3
    Alpha=1
    ConstraintWeights_0=0.0348
    ConstraintWeights_1:float=0.1233
    ConstraintWeights_2:float=0.271
    ConstraintWeights_3:float=0.0713
    ConstraintWeights_4:float=0.0152
    ConstraintWeights_5:float=0.4563
    ConstraintWeights_6:float=0.0282
   ) -> (Pose) {
    alphaBoolBlend {blendOption=Linear}
    boneToModify {boneName=master_L_dring_jnt}
    constraintSetup [
     {targetBone {boneName=spine_01}}
     {targetBone {boneName=spine_02}}
     {targetBone {boneName=spine_04}}
     {targetBone {boneName=spine_05}}
     {targetBone {boneName=clavicle_l}}
     {targetBone {boneName=spine_03}}
     {targetBone {boneName=clavicle_r}}
    ]
    constraintWeights [1,1,1,1,1,1,1]
   }
   n5 = "Constraint - master_R_dring_jnt (PPPPPP)"!(
    ComponentPose=@n4
    Alpha=1
    ConstraintWeights_0=0.0389
    ConstraintWeights_1:float=0.1099
    ConstraintWeights_2:float=0.2378
    ConstraintWeights_3:float=0.0472
    ConstraintWeights_4:float=0.365
    ConstraintWeights_5:float=0.2011
   ) -> (Pose) {
    alphaBoolBlend {blendOption=Linear}
    boneToModify {boneName=master_R_dring_jnt}
    constraintSetup [
     {targetBone {boneName=spine_01}}
     {targetBone {boneName=spine_02}}
     {targetBone {boneName=spine_04}}
     {targetBone {boneName=spine_05}}
     {targetBone {boneName=spine_03}}
     {targetBone {boneName=clavicle_r}}
    ]
    constraintWeights [1,1,1,1,1,1]
   }
   n6 = "Constraint - master_chain_jnt (PPPPP)"!(
    ComponentPose=@n5
    Alpha=1
    ConstraintWeights_0=0.4301
    ConstraintWeights_1:float=0.4102
    ConstraintWeights_2:float=0.0212
    ConstraintWeights_3:float=0.126
    ConstraintWeights_4:float=0.0125
   ) -> (Pose) {
    alphaBoolBlend {blendOption=Linear}
    boneToModify {boneName=master_chain_jnt}
    constraintSetup [
     {targetBone {boneName=pelvis}}
     {targetBone {boneName=thigh_twist_01_l}}
     {targetBone {boneName=thigh_twist_01_r}}
     {targetBone {boneName=deform_groin_l}}
     {targetBone {boneName=deform_groin_r}}
    ]
    constraintWeights [1,1,1,1,1]
   }
   n7 = "Transform (Modify) Bone - Bone: master_chain_jnt"(ComponentPose=@n6, Alpha=Get Crouching() -> (Crouching:double), Translation="-5.260,1.310,-2.337", Rotation="-31,-29,0", Scale="0,1,1") {
    rotationMode=BMM_Additive
    rotationSpace=BCS_BoneSpace
    translationMode=BMM_Additive
    translationSpace=BCS_BoneSpace
    alphaBoolBlend {blendOption=Linear}
    boneToModify {boneName=master_chain_jnt}
    rotation {pitch=-31, yaw=-29}
    scale {x=0}
    translation=<-5.26 1.31 -2.337>
   }
   n8 = RigidBody(ComponentPose=@n7) {
    bClampLinearTranslationLimitToRefPose=true
    bOverrideWorldGravity=true
    overridePhysicsAsset=&/M_MED_BlueJet_Physics
    simulationSpace=BaseBoneSpace
    alphaBoolBlend {blendOption=Linear}
    baseBoneRef {boneName=spine_03}
    componentAppliedLinearAccClamp=<3000 3000 3000>
    componentLinearAccScale=<0.3 0.2 0.2>
    componentLinearVelScale=<0.3 0.2 0.2>
    overrideWorldGravity {z=-980}
    simSpaceSettings {maxAngularAcceleration=1000, maxAngularVelocity=3, maxLinearAcceleration=250, maxLinearVelocity=500}
   }
   n9 = Anim Dynamics - Bone: dyn_L_dring_jnt(ComponentPose=@n8, Alpha=1, ExternalForce="0,0,0") {
    angularSpringConstant=20000
    bAngularSpring=true
    bUseGravityOverride=true
    simulationSpace=RootRelative
    alphaBoolBlend {blendOption=Linear}
    boundBone {boneName=dyn_L_dring_jnt}
    componentAppliedLinearAccClamp=<3000 3000 3000>
    componentLinearAccScale=<0.3 0.3 0.3>
    componentLinearVelScale=<0.3 0.3 0.3>
    gravityOverride {z=-980}
    physicsBodyDefinitions [{
      boundBone {boneName=dyn_L_dring_jnt}
      boxExtents=<10 10 10>
      constraintSetup {angularLimitsMax=<0 0 5>, angularLimitsMin=<0 0 -60>, angularTarget=<0 0 0>}
      localJointOffset=<5 0 0>
     }]
    relativeSpaceBone {boneName=master_L_dring_jnt}
   }
   n10 = Anim Dynamics - Bone: dyn_R_dring_jnt(ComponentPose=@n9, Alpha=1, ExternalForce="0,0,0") {
    angularSpringConstant=20000
    bAngularSpring=true
    bUseGravityOverride=true
    simulationSpace=RootRelative
    alphaBoolBlend {blendOption=Linear}
    boundBone {boneName=dyn_R_dring_jnt}
    componentAppliedLinearAccClamp=<3000 3000 3000>
    componentLinearAccScale=<0.4 0.4 0.4>
    componentLinearVelScale=<0.4 0.4 0.4>
    gravityOverride {z=-980}
    physicsBodyDefinitions [{
      boundBone {boneName=dyn_R_dring_jnt}
      boxExtents=<10 10 10>
      constraintSetup {angularLimitsMax=<0 0 5>, angularLimitsMin=<0 0 -60>, angularTarget=<0 0 0>}
      localJointOffset=<-5 0 0>
     }]
    relativeSpaceBone {boneName=master_L_dring_jnt}
   }
   n11 = "Transform (Modify) Bone - Bone: master_R_buckle_jnt"(ComponentPose=@n10, Alpha=1, Translation="0,0,0", Rotation="0,7,0", Scale="1,1,1") {
    rotationMode=BMM_Additive
    rotationSpace=BCS_BoneSpace
    alphaBoolBlend {blendOption=Linear}
    boneToModify {boneName=master_R_buckle_jnt}
    rotation {yaw=7}
   }
   n12 = "Constraint - dyn_chain_jnt (P)"(ComponentPose=@n11, Alpha=Get Emoting() -> (Emoting:double), ConstraintWeights_0=1) {
    alphaBoolBlend {blendOption=Linear}
    boneToModify {boneName=dyn_chain_jnt}
    constraintSetup [{targetBone {boneName=thigh_twist_01_l}}]
    constraintWeights [1]
   }
   n13 = "To Float (Boolean)"(InBool=NOT Boolean(A=Get bAnimDynamicsEnabled() -> (bAnimDynamicsEnabled:bool)))
   n14 = "Transform (Modify) Bone - Bone: master_L_buckle_jnt"(ComponentPose=@n12, Alpha=@n13, Translation="0,0,0", Rotation="0,-6,0", Scale="1,1,1") {
    rotationMode=BMM_Additive
    rotationSpace=BCS_BoneSpace
    alphaBoolBlend {blendOption=Linear}
    boneToModify {boneName=master_L_buckle_jnt}
    rotation {yaw=-6}
   }
   n15 = "Transform (Modify) Bone - Bone: master_R_buckle_jnt"(ComponentPose=@n14, Alpha=@n13, Translation="0,0,0", Rotation="0,-6,0", Scale="1,1,1") {
    rotationMode=BMM_Additive
    rotationSpace=BCS_BoneSpace
    alphaBoolBlend {blendOption=Linear}
    boneToModify {boneName=master_R_buckle_jnt}
    rotation {yaw=-6}
   }
   Output Pose(Result=Component To Local(ComponentPose=@n15)) {layerGroup=None}
  ]}
 EventGraph {nodes [
   Event BlueprintUpdateAnimation()
   > Set Crouching(Crouching:double="To Float (Boolean)"(InBool=Get bIsCrouching())) -> (Output_Get:double)
   > Set FrontEnd(FrontEnd:double="To Float (Boolean)"(InBool=Get bIsInFrontEnd())) -> (Output_Get:double)
   > Set Emoting(Emoting:double="To Float (Boolean)"(InBool=Get bIsPlayingEmote())) -> (Output_Get:double)
  ]}
]
variables [
 Crouching {type=number}
 Emoting {type=number}
 FrontEnd {type=number}
]
