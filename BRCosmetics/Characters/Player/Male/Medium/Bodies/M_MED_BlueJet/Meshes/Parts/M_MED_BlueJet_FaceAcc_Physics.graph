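# Physics asset for the BlueJet medium male body's face accessory mesh, covering its hair fringe bones.
# Capsule bodies on the neck and head link up through master hair and bangs joints with free angular constraints.
# From the bangs joint, a two-bone left chain and a right chain forking into front and back tips branch off.
# These fringe joints have narrow swing and tighter twist limits.
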PhysicsAsset /BRCosmetics/Characters/Player/Male/Medium/Bodies/M_MED_BlueJet/Meshes/Parts/M_MED_BlueJet_FaceAcc_Physics
bodies {
 dyn_L_hair_fringe_1_jnt [{
   shapeType=Capsule
   length=1.5
   radius=1.5
   shapeName=dyn_L_hair_fringe_1_jnt_capsule
   center=<1.798 0 1.5466>
   rotation=<rot 90 180 179.9995>
  }]
 dyn_L_hair_fringe_2_jnt [{
   shapeType=Capsule
   length=2
   radius=1.5
   shapeName=dyn_L_hair_fringe_2_jnt_capsule
   center=<2.382 0 1.1069>
   rotation=<rot 80 0.0006 -0.0001>
  }]
 dyn_R_hair_fringe_1_jnt [{
   shapeType=Capsule
   length=1.5
   radius=1.75
   shapeName=dyn_R_hair_fringe_1_jnt_capsule
   center=<1.6466 0 -0.0109>
   rotation=<rot 80 179.9996 179.9996>
  }]
 dyn_R_hair_fringe_2_jnt [{
   shapeType=Capsule
   length=1.5
   radius=1.75
   shapeName=dyn_R_hair_fringe_2_jnt_capsule
   center=<1.7141 0 -0.3936>
   rotation=<rot 80 -179.9999 179.9998>
  }]
 dyn_R_hair_fringe_back_jnt [{
   shapeType=Capsule
   length=3
   radius=1.75
   shapeName=dyn_R_hair_fringe_back_jnt_capsule
   center=<2.9599 -0.8631 0.5219>
   rotation=<rot 80 179.9996 -160.0007>
  }]
 dyn_R_hair_fringe_front_jnt [{
   shapeType=Capsule
   length=1.5
   radius=1.75
   shapeName=dyn_R_hair_fringe_front_jnt_capsule
   center=<1.5951 0 0.2813>
   rotation=<rot 80 180 179.9998>
  }]
 head [
  {shapeType=Capsule, length=6.25, radius=6, shapeName=head_capsule, center=<4.5735 4.8829 0>, rotation=<rot 90 0 0>}
  {shapeType=Capsule, length=21, radius=4, shapeName=head_capsule, center=<7.7188 1.8089 0>}
 ]
 master_bangs_jnt [{shapeType=Capsule, length=1, radius=1, shapeName=master_bangs_jnt_capsule}]
 master_hair_jnt [{shapeType=Capsule, length=1, radius=1, shapeName=master_hair_jnt_capsule}]
 neck_02 [{shapeType=Capsule, length=3.25, radius=3.75, shapeName=neck_02_capsule, center=<1.4298 0 0>, rotation=<rot 90 0 0>}]
}
constraints [
 {bone1Name=dyn_R_hair_fringe_front_jnt, bone2Name=dyn_R_hair_fringe_2_jnt, swing1LimitDegrees=11, swing2LimitDegrees=11, twistLimitDegrees=2}
 {
  bone1Name=head
  bone2Name=neck_02
  swing1LimitDegrees=45
  swing1Motion=Free
  swing2LimitDegrees=45
  swing2Motion=Free
  twistLimitDegrees=45
  twistMotion=Free
 }
 {
  bone1Name=master_hair_jnt
  bone2Name=head
  swing1LimitDegrees=45
  swing1Motion=Free
  swing2LimitDegrees=45
  swing2Motion=Free
  twistLimitDegrees=45
  twistMotion=Free
 }
 {
  bone1Name=master_bangs_jnt
  bone2Name=master_hair_jnt
  swing1LimitDegrees=45
  swing1Motion=Free
  swing2LimitDegrees=45
  swing2Motion=Free
  twistLimitDegrees=45
  twistMotion=Free
 }
 {bone1Name=dyn_L_hair_fringe_1_jnt, bone2Name=master_bangs_jnt, swing1LimitDegrees=11, swing2LimitDegrees=11, twistLimitDegrees=2}
 {bone1Name=dyn_L_hair_fringe_2_jnt, bone2Name=dyn_L_hair_fringe_1_jnt, swing1LimitDegrees=11, swing2LimitDegrees=11, twistLimitDegrees=2}
 {bone1Name=dyn_R_hair_fringe_1_jnt, bone2Name=master_bangs_jnt, swing1LimitDegrees=11, swing2LimitDegrees=11, twistLimitDegrees=2}
 {bone1Name=dyn_R_hair_fringe_2_jnt, bone2Name=dyn_R_hair_fringe_1_jnt, swing1LimitDegrees=11, swing2LimitDegrees=11, twistLimitDegrees=2}
 {bone1Name=dyn_R_hair_fringe_back_jnt, bone2Name=dyn_R_hair_fringe_2_jnt, swing1LimitDegrees=11, swing2LimitDegrees=11, twistLimitDegrees=2}
]
